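# Physics asset for the LocalZilla glider mesh, a VinderTech GliderChute gadget cosmetic, giving its wings simple physics shapes.
# It holds two bodies, one capsule each on the left and right wing base bones, with equal size and mirrored center offsets.
# No constraints join the bodies, and no other bones have bodies.
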PhysicsAsset /BRCosmetics/Gadgets/Assets/VinderTech_GliderChute/Glider_LocalZilla/Meshes/LocalZilla_Glider_PhysicsAsset
bodies {
 Wing_Base_l [{shapeType=Capsule, length=65.0138, radius=18.3259, center=<-72.6257 -38.3663 -12.1859>, rotation=<rot 0 -85.7328 -85.5189>}]
 Wing_Base_r [{shapeType=Capsule, length=65.0138, radius=18.3259, center=<72.6257 38.3663 12.1859>, rotation=<rot 0 -85.7328 -85.5189>}]
}
constraints []
